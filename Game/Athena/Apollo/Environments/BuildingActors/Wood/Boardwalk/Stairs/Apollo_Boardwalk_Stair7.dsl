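; Boardwalk wooden stair building actor that plays a looping wood creak sound in the environment.
; On BeginPlay it sets the sound on its WoodCreakSound audio component to the WoodCreakLoop sound cue setting variable.
; The construction script is empty.
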(event EventBeginPlay
  (Audio|Components|Audio|SetSound (Variables|SM_Boardwalk_Stair34|GetWoodCreakSound) (Variables|Default|GetWoodCreakLoop_SoundCueSetting)))

(fn ConstructionScript)
